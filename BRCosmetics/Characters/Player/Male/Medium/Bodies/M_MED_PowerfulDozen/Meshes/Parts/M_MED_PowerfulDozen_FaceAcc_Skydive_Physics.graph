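# Physics asset for the PowerfulDozen medium male face accessory in skydive, covering two dangling headband tails and the upper body.
# Seven left and six right capsules form chains with free swing and tight twist limits, rooted through zero-size master bodies to the head.
# Larger capsules on head, neck, clavicles and upper spine are joined by fully free constraints.
# One constraint references dyn_turtleShell_01, which has no body here.
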PhysicsAsset /BRCosmetics/Characters/Player/Male/Medium/Bodies/M_MED_PowerfulDozen/Meshes/Parts/M_MED_PowerfulDozen_FaceAcc_Skydive_Physics
bodies {
 HeadBand_MSTR [{
   shapeType=Capsule
   length=0
   radius=0
   shapeName=HeadBand_MSTR_capsule
   center=<0.6008 -0.3292 -0.7705>
   rotation=<rot -13.7906 -90 -40.2646>
  }]
 clavicle_l [{shapeType=Capsule, length=18, radius=10, shapeName=clavicle_l_capsule, center=<8 -2 2>, rotation=<rot 80 -0.0001 -0.0002>}]
 clavicle_r [{shapeType=Capsule, length=18, radius=10, shapeName=clavicle_r_capsule, center=<-8 2 -2>, rotation=<rot 80 -0.0001 -0.0002>}]
 dyn_HeadBand_l_01_jnt [{shapeType=Capsule, length=5, radius=2, shapeName=dyn_HeadBand_l_01_jnt_capsule, center=<4 0 0>, rotation=<rot 90 0 0>}]
 dyn_HeadBand_l_02_jnt [{shapeType=Capsule, length=5, radius=2, shapeName=dyn_HeadBand_l_02_jnt_capsule, center=<4 0 0>, rotation=<rot 90 0 0>}]
 dyn_HeadBand_l_03_jnt [{shapeType=Capsule, length=5, radius=2, shapeName=dyn_HeadBand_l_03_jnt_capsule, center=<4 0 0>, rotation=<rot 90 0 0>}]
 dyn_HeadBand_l_04_jnt [{shapeType=Capsule, length=5, radius=2, shapeName=dyn_HeadBand_l_04_jnt_capsule, center=<4 0 0>, rotation=<rot 90 0 0>}]
 dyn_HeadBand_l_05_jnt [{shapeType=Capsule, length=5, radius=2, shapeName=dyn_HeadBand_l_05_jnt_capsule, center=<4 0 0>, rotation=<rot 90 0 0>}]
 dyn_HeadBand_l_06_jnt [{shapeType=Capsule, length=5, radius=2, shapeName=dyn_HeadBand_l_06_jnt_capsule, center=<4 0 0>, rotation=<rot 90 0 0>}]
 dyn_HeadBand_l_07_jnt [{shapeType=Capsule, length=5, radius=2, shapeName=dyn_HeadBand_l_07_jnt_capsule, center=<4 0 0>, rotation=<rot 90 0 0>}]
 dyn_HeadBand_l_MSTR [{shapeType=Capsule, length=0, radius=0, shapeName=dyn_HeadBand_l_MSTR_capsule, rotation=<rot 0 -90 0>}]
 dyn_HeadBand_r_01_jnt [{shapeType=Capsule, length=5, radius=2, shapeName=dyn_HeadBand_r_01_jnt_capsule, center=<4 0 0>, rotation=<rot 90 0 0>}]
 dyn_HeadBand_r_02_jnt [{shapeType=Capsule, length=5, radius=2, shapeName=dyn_HeadBand_r_02_jnt_capsule, center=<4 0 0>, rotation=<rot 90 0 0>}]
 dyn_HeadBand_r_03_jnt [{shapeType=Capsule, length=5, radius=1, shapeName=dyn_HeadBand_r_03_jnt_capsule, center=<4 0 0>, rotation=<rot 90 0 0>}]
 dyn_HeadBand_r_04_jnt [{shapeType=Capsule, length=5, radius=1, shapeName=dyn_HeadBand_r_04_jnt_capsule, center=<4 0 0>, rotation=<rot 90 0 0>}]
 dyn_HeadBand_r_05_jnt [{shapeType=Capsule, length=5, radius=1, shapeName=dyn_HeadBand_r_05_jnt_capsule, center=<4 0 0>, rotation=<rot 90 0 0>}]
 dyn_HeadBand_r_06_jnt [{shapeType=Capsule, length=5, radius=1, shapeName=dyn_HeadBand_r_06_jnt_capsule, center=<4 0 0>, rotation=<rot 90 0 0>}]
 dyn_HeadBand_r_MSTR [{shapeType=Capsule, length=0, radius=0, shapeName=dyn_HeadBand_r_MSTR_capsule, rotation=<rot 0 -90 0>}]
 head [{shapeType=Capsule, length=4, radius=10, shapeName=head_capsule, center=<5 2 0>}]
 neck_01 [{shapeType=Capsule, length=8, radius=8, shapeName=neck_01_capsule, center=<3.6683 -2 0>, rotation=<rot 90 0 0>}]
 spine_03 [{shapeType=Capsule, length=15, radius=10, shapeName=spine_04_capsule, center=<0.229 2 0>, rotation=<rot 90 0 10.0001>}]
 spine_04 [{shapeType=Capsule, length=12, radius=14, shapeName=spine_04_capsule, center=<0 1 0>}]
 spine_05 [{shapeType=Capsule, length=16, radius=14, shapeName=spine_05_capsule, center=<-1 0.3959 0>}]
}
constraints [
 {
  bone1Name=HeadBand_MSTR
  bone2Name=head
  swing1LimitDegrees=45
  swing1Motion=Free
  swing2LimitDegrees=45
  swing2Motion=Free
  twistLimitDegrees=45
  twistMotion=Free
 }
 {
  bone1Name=head
  bone2Name=neck_01
  swing1LimitDegrees=45
  swing1Motion=Free
  swing2LimitDegrees=45
  swing2Motion=Free
  twistLimitDegrees=45
  twistMotion=Free
 }
 {
  bone1Name=head
  bone2Name=spine_05
  swing1LimitDegrees=45
  swing1Motion=Free
  swing2LimitDegrees=45
  swing2Motion=Free
  twistLimitDegrees=45
  twistMotion=Free
 }
 {
  bone1Name=spine_04
  bone2Name=spine_03
  swing1LimitDegrees=45
  swing1Motion=Free
  swing2LimitDegrees=45
  swing2Motion=Free
  twistLimitDegrees=45
  twistMotion=Free
 }
 {
  bone1Name=spine_05
  bone2Name=spine_04
  swing1LimitDegrees=45
  swing1Motion=Free
  swing2LimitDegrees=45
  swing2Motion=Free
  twistLimitDegrees=45
  twistMotion=Free
 }
 {
  bone1Name=dyn_turtleShell_01
  bone2Name=spine_04
  swing1LimitDegrees=45
  swing1Motion=Free
  swing2LimitDegrees=45
  swing2Motion=Free
  twistLimitDegrees=45
  twistMotion=Free
 }
 {
  bone1Name=neck_01
  bone2Name=spine_05
  swing1LimitDegrees=45
  swing1Motion=Free
  swing2LimitDegrees=45
  swing2Motion=Free
  twistLimitDegrees=45
  twistMotion=Free
 }
 {
  bone1Name=clavicle_l
  bone2Name=spine_05
  swing1LimitDegrees=45
  swing1Motion=Free
  swing2LimitDegrees=45
  swing2Motion=Free
  twistLimitDegrees=45
  twistMotion=Free
 }
 {
  bone1Name=clavicle_r
  bone2Name=spine_05
  swing1LimitDegrees=45
  swing1Motion=Free
  swing2LimitDegrees=45
  swing2Motion=Free
  twistLimitDegrees=45
  twistMotion=Free
 }
 {
  bone1Name=dyn_HeadBand_l_MSTR
  bone2Name=HeadBand_MSTR
  swing1LimitDegrees=45
  swing1Motion=Free
  swing2LimitDegrees=45
  swing2Motion=Free
  twistLimitDegrees=45
  twistMotion=Free
 }
 {
  bone1Name=dyn_HeadBand_l_01_jnt
  bone2Name=dyn_HeadBand_l_MSTR
  swing1LimitDegrees=35
  swing1Motion=Free
  swing2LimitDegrees=35
  swing2Motion=Free
  twistLimitDegrees=2
 }
 {
  bone1Name=dyn_HeadBand_l_02_jnt
  bone2Name=dyn_HeadBand_l_01_jnt
  swing1LimitDegrees=35
  swing1Motion=Free
  swing2LimitDegrees=35
  swing2Motion=Free
  twistLimitDegrees=2
 }
 {
  bone1Name=dyn_HeadBand_l_03_jnt
  bone2Name=dyn_HeadBand_l_02_jnt
  swing1LimitDegrees=35
  swing1Motion=Free
  swing2LimitDegrees=35
  swing2Motion=Free
  twistLimitDegrees=2
 }
 {
  bone1Name=dyn_HeadBand_l_04_jnt
  bone2Name=dyn_HeadBand_l_03_jnt
  swing1LimitDegrees=35
  swing1Motion=Free
  swing2LimitDegrees=35
  swing2Motion=Free
  twistLimitDegrees=2
 }
 {
  bone1Name=dyn_HeadBand_l_05_jnt
  bone2Name=dyn_HeadBand_l_04_jnt
  swing1LimitDegrees=45
  swing1Motion=Free
  swing2LimitDegrees=45
  swing2Motion=Free
  twistLimitDegrees=2
 }
 {
  bone1Name=dyn_HeadBand_l_06_jnt
  bone2Name=dyn_HeadBand_l_05_jnt
  swing1LimitDegrees=45
  swing1Motion=Free
  swing2LimitDegrees=45
  swing2Motion=Free
  twistLimitDegrees=2
 }
 {
  bone1Name=dyn_HeadBand_l_07_jnt
  bone2Name=dyn_HeadBand_l_06_jnt
  swing1LimitDegrees=45
  swing1Motion=Free
  swing2LimitDegrees=45
  swing2Motion=Free
  twistLimitDegrees=2
 }
 {
  bone1Name=dyn_HeadBand_r_MSTR
  bone2Name=HeadBand_MSTR
  swing1LimitDegrees=45
  swing1Motion=Free
  swing2LimitDegrees=45
  swing2Motion=Free
  twistLimitDegrees=45
  twistMotion=Free
 }
 {
  bone1Name=dyn_HeadBand_r_01_jnt
  bone2Name=dyn_HeadBand_r_MSTR
  swing1LimitDegrees=35
  swing1Motion=Free
  swing2LimitDegrees=35
  swing2Motion=Free
  twistLimitDegrees=2
 }
 {
  bone1Name=dyn_HeadBand_r_02_jnt
  bone2Name=dyn_HeadBand_r_01_jnt
  swing1LimitDegrees=35
  swing1Motion=Free
  swing2LimitDegrees=35
  swing2Motion=Free
  twistLimitDegrees=2
 }
 {
  bone1Name=dyn_HeadBand_r_03_jnt
  bone2Name=dyn_HeadBand_r_02_jnt
  swing1LimitDegrees=35
  swing1Motion=Free
  swing2LimitDegrees=35
  swing2Motion=Free
  twistLimitDegrees=2
 }
 {
  bone1Name=dyn_HeadBand_r_04_jnt
  bone2Name=dyn_HeadBand_r_03_jnt
  swing1LimitDegrees=35
  swing1Motion=Free
  swing2LimitDegrees=35
  swing2Motion=Free
  twistLimitDegrees=2
 }
 {
  bone1Name=dyn_HeadBand_r_05_jnt
  bone2Name=dyn_HeadBand_r_04_jnt
  swing1LimitDegrees=45
  swing1Motion=Free
  swing2LimitDegrees=45
  swing2Motion=Free
  twistLimitDegrees=2
 }
 {
  bone1Name=dyn_HeadBand_r_06_jnt
  bone2Name=dyn_HeadBand_r_05_jnt
  swing1LimitDegrees=45
  swing1Motion=Free
  swing2LimitDegrees=45
  swing2Motion=Free
  twistLimitDegrees=2
 }
]
